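# Sound cue for the Hot Feet bounce of the flame trap, routed to the player foley sound class.
# An IsLocal branch plays one of five equally weighted first-person bounce waves for the local player, with no modulator or attenuation node.
# Otherwise, or when the parameter is unset, it plays one of five sizzle waves through a modulator and the shared Hot Feet bubble burst attenuation.
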SoundCue /FNGameplayCues/Sounds/Fort_Traps/Flame_Trap/Hot_Feet_Bounce_Cue
root=Branch_0
nodes [
 {
  id=Attenuation_0
  children [Modulator_0]
  properties {attenuationSettings=&/Hot_Feet_Bubble_Burst_Att}
 }
 {id=Branch_0, children [Random_2, Attenuation_0, Attenuation_0], properties {boolParameterName=IsLocal}}
 {id=Modulator_0, children [Random_0]}
 {
  id=Random_0
  children [WavePlayer_1,WavePlayer_2,WavePlayer_3,WavePlayer_4,WavePlayer_6]
  properties {bShouldExcludeFromBranchCulling=true, weights [1,1,1,1,1]}
 }
 {
  id=Random_2
  children [WavePlayer_7, WavePlayer_11, WavePlayer_12, WavePlayer_13, WavePlayer_14]
  properties {bShouldExcludeFromBranchCulling=true, weights [1,1,1,1,1]}
 }
 {id=WavePlayer_1, properties {soundWaveAssetPtr=&/Hot_Feet_Bounce_Sizzle_01}}
 {id=WavePlayer_11, properties {soundWaveAssetPtr=&/Hot_Feet_Bounce_1P_02}}
 {id=WavePlayer_12, properties {soundWaveAssetPtr=&/Hot_Feet_Bounce_1P_03}}
 {id=WavePlayer_13, properties {soundWaveAssetPtr=&/Hot_Feet_Bounce_1P_04}}
 {id=WavePlayer_14, properties {soundWaveAssetPtr=&/Hot_Feet_Bounce_1P_05}}
 {id=WavePlayer_2, properties {soundWaveAssetPtr=&/Hot_Feet_Bounce_Sizzle_02}}
 {id=WavePlayer_3, properties {soundWaveAssetPtr=&/Hot_Feet_Bounce_Sizzle_03}}
 {id=WavePlayer_4, properties {soundWaveAssetPtr=&/Hot_Feet_Bounce_Sizzle_04}}
 {id=WavePlayer_6, properties {soundWaveAssetPtr=&/Hot_Feet_Bounce_Sizzle_05}}
 {id=WavePlayer_7, properties {soundWaveAssetPtr=&/Hot_Feet_Bounce_1P_01}}
]
properties {maxDistance=6400, soundClassObject=&Fort_Player_Foley, volumeMultiplier=2}
