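# Sound cue for the flame trap's Hot Feet bubble burst, played as a spatialized sound in the player foley sound class.
# A random node picks one of five equally weighted burst-burn waves and is excluded from branch culling.
# The pick passes through a modulator for variation, then into a boosted mixer input; the mixer's other input is empty.
# A dedicated attenuation asset handles distance falloff, and the cue's overall volume is raised.
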SoundCue /FNGameplayCues/Sounds/Fort_Traps/Flame_Trap/Hot_Feet_Bubble_Burst_Cue
root=Mixer_0
nodes [
 {id=Mixer_0, children [null,Modulator_0], properties {inputVolume [1,2]}}
 {id=Modulator_0, children [Random_1], properties {volumeMax=1}}
 {
  id=Random_1
  children [WavePlayer_0,WavePlayer_5,WavePlayer_7,WavePlayer_8,WavePlayer_9]
  properties {bShouldExcludeFromBranchCulling=true, weights [1,1,1,1,1]}
 }
 {id=WavePlayer_0, properties {soundWaveAssetPtr=&/Hot_Feet_Burst_Burn_01}}
 {id=WavePlayer_5, properties {soundWaveAssetPtr=&/Hot_Feet_Burst_Burn_02}}
 {id=WavePlayer_7, properties {soundWaveAssetPtr=&/Hot_Feet_Burst_Burn_03}}
 {id=WavePlayer_8, properties {soundWaveAssetPtr=&/Hot_Feet_Burst_Burn_04}}
 {id=WavePlayer_9, properties {soundWaveAssetPtr=&/Hot_Feet_Burst_Burn_05}}
]
properties {
 attenuationSettings=&/Hot_Feet_Bubble_Burst_Att
 maxDistance=6400
 soundClassObject=&Fort_Player_Foley
 volumeMultiplier=1.5
}
