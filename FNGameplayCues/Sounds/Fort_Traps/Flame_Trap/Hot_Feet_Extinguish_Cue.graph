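# Sound cue for the Hot Feet extinguish sound of the flame trap.
# A random node picks one of four extinguish waves at equal weight and feeds a modulator at the root.
# It plays at reduced volume and borrows the small melee weapon attenuation and the melee weapon sound class.
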SoundCue /FNGameplayCues/Sounds/Fort_Traps/Flame_Trap/Hot_Feet_Extinguish_Cue
root=Modulator_2
nodes [
 {id=Modulator_2, children [Random_0]}
 {id=Random_0, children [WavePlayer_1,WavePlayer_2,WavePlayer_3,WavePlayer_4], properties {weights [1,1,1,1]}}
 {id=WavePlayer_1, properties {soundWaveAssetPtr=/Game/Sounds/Fort_Traps/Flame_Trap/Hot_Feet_Extinguish_01}}
 {id=WavePlayer_2, properties {soundWaveAssetPtr=&/Hot_Feet_Extinguish_02}}
 {id=WavePlayer_3, properties {soundWaveAssetPtr=&/Hot_Feet_Extinguish_03}}
 {id=WavePlayer_4, properties {soundWaveAssetPtr=&/Hot_Feet_Extinguish_04}}
]
properties {
 attenuationSettings=&Fort_Weapon_Melee_Sml_Att
 maxDistance=2800
 soundClassObject=&Fort_WeaponMelee_PSM
 volumeMultiplier=0.7
}
